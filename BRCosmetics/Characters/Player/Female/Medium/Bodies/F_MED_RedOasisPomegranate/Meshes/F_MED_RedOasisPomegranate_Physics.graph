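# Physics asset for the RedOasisPomegranate medium female player body cosmetic, set up for secondary motion on the left and right shorts hems.
# Capsule bodies cover the pelvis, both thighs and one dynamic hem bone per leg.
# Each hem is constrained to its thigh with small swing and twist limits, while the thighs have zero-degree limits to the pelvis.
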PhysicsAsset /BRCosmetics/Characters/Player/Female/Medium/Bodies/F_MED_RedOasisPomegranate/Meshes/F_MED_RedOasisPomegranate_Physics
bodies {
 dyn_shorts_hem_l [{shapeType=Capsule, length=2, radius=4, shapeName=dyn_shorts_hem_l_capsule, center=<6 0 0>, rotation=<rot 90 0 0>}]
 dyn_shorts_hem_r [{shapeType=Capsule, length=2, radius=4, shapeName=dyn_shorts_hem_r_capsule, center=<6 0 0>, rotation=<rot 90 0 0>}]
 pelvis [{
   shapeType=Capsule
   length=7.3578
   radius=3.5055
   shapeName=pelvis_capsule
   center=<-1.4684 -2.2286 -0.0344>
   rotation=<rot -0.9289 -90 -1.3008>
  }]
 thigh_l [{shapeType=Capsule, length=3.9957, radius=4.6084, shapeName=thigh_l_capsule, center=<-10 0 0>, rotation=<rot 90 0 0>}]
 thigh_r [{shapeType=Capsule, length=4.214, radius=4.3901, shapeName=thigh_r_capsule, center=<10 0 0>, rotation=<rot 90 0 0>}]
}
constraints [
 {bone1Name=dyn_shorts_hem_l, bone2Name=thigh_l, swing1LimitDegrees=3, swing2LimitDegrees=8, twistLimitDegrees=12}
 {bone1Name=dyn_shorts_hem_r, bone2Name=thigh_r, swing1LimitDegrees=3, swing2LimitDegrees=8, twistLimitDegrees=12}
 {bone1Name=thigh_l, bone2Name=pelvis, swing1LimitDegrees=0, swing2LimitDegrees=0, twistLimitDegrees=0}
 {bone1Name=thigh_r, bone2Name=pelvis, swing1LimitDegrees=0, swing2LimitDegrees=0, twistLimitDegrees=0}
]
